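# Mono one-shot MetaSound source for ColdFern weapon foley, under pickups, using weapon-reload attenuation and the weapon foley sound class.
# On play, it picks a random wave from an input array, then rolls a random pitch shift between the Min and Max inputs before starting the wave player.
# Output is scaled by a Multiplier input, and the source finishes when the wave ends.
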MetaSoundSource /WildEstateWeapons/Gameplay/ColdFern/Sound/Pickups/MSS_ColdFern_Foley_lib
inputs [
 In Array {type=WaveAsset:Array}
 Multiplier {type=Float}
 Pitch.Max {type=Float}
 Pitch.Min {type=Float}
 UE.Source.OnPlay {type=Trigger}
]
interfaces [UE.OutputFormat.Mono:1.0, UE.Source.OneShot:1.0, UE.Source:1.1]
outputs [
 "UE.OutputFormat.Mono.Audio:0" {type=Audio}
 UE.Source.OneShot.OnFinished {type=Trigger}
]
pages [Default {
  input_defaults {In Array=null, Multiplier=1, Pitch.Max=0.5, Pitch.Min=-0.5, UE.Source.OnPlay=null}
  nodes [
   ArrayRandomGet_0 = Array.Random Get:WaveAsset:Array(In Array=^In Array, Next=^UE.Source.OnPlay) {also {defaults {Weights []}}}
   n1 = UE.RandomFloat(Max=^Pitch.Max, Min=^Pitch.Min, Next=@ArrayRandomGet_0.On Next, Reset=@ArrayRandomGet_0.On Reset)
   WavePlayer_0 = UE.Wave Player:Mono(Wave Asset=@ArrayRandomGet_0.Value, Play=@n1.On Next, Pitch Shift=@n1.Value) {also {version [1,1]}}
   Multiply_0 = UE.Multiply:Audio by Float(AdditionalOperands=^Multiplier, PrimaryOperand=@WavePlayer_0.Out Mono)
  ]
  links [{dst="@UE.OutputFormat.Mono.Audio:0", src=Multiply_0.Out}, {dst="@UE.Source.OneShot.OnFinished", src=WavePlayer_0.On Finished}]
 }]
properties {
 assetImportData=&self:AssetImportData
 attenuationSettings=&Fort_Weapon_Reload_Att
 maxDistance=2250
 sampleRate=48000
 soundClassObject=&Fort_WeaponFoley
}
