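# MetaSound source for the ColdFern weapon dry-fire sound, a mono one-shot in the weapon foley sound class.
# On play, it passes an array of four pistol, revolver and Uzi dry-fire waves to the shared MSS_ColdFern_Foley_lib graph, which supplies the audio and finished trigger.
# A "Can Not Do" error-ping input is declared but left unconnected.
# It uses weapon reload attenuation and overrides concurrency to limit simultaneous instances.
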MetaSoundSource /WildEstateWeapons/Gameplay/ColdFern/Sound/Pickups/MSS_ColdFern_DryFire
inputs [
 Can Not Do {type=WaveAsset}
 Dry Fire {type=WaveAsset:Array}
 Time.ElapsedTime {type=Float}
 Time.GameTimeSeconds {type=Float}
 Time.PlayTime {type=Float}
 Time.TimeSinceLastPlayed {type=Float}
 UE.Source.OnPlay {type=Trigger}
]
interfaces [Fort.Time:1.0, UE.OutputFormat.Mono:1.0, UE.Source.OneShot:1.0, UE.Source:1.1]
outputs [
 "UE.OutputFormat.Mono.Audio:0" {type=Audio}
 UE.Source.OneShot.OnFinished {type=Trigger}
]
pages [Default {
  input_defaults {
   Can Not Do=/Game/Sounds/Fort_UI_Sounds/Gameplay/UI_Error_Ping_Cant_Do_Default_01
   Time.ElapsedTime=0
   Time.GameTimeSeconds=0
   Time.PlayTime=0
   Time.TimeSinceLastPlayed=1
   UE.Source.OnPlay=null
   Dry Fire [
    /WildEstateWeapons/Gameplay/ColdFern/Sound/wav/ColdFern_Wpn_1911_Fire_1P_DryFire
    /WildEstateWeapons/Gameplay/ColdFern/Sound/wav/ColdFern_Wpn_Revolver_Fire_1P_DryFire
    /WildEstateWeapons/Gameplay/ColdFern/Sound/wav/ColdFern_Wpn_PPK_Fire_1P_DryFire
    /WildEstateWeapons/Gameplay/ColdFern/Sound/wav/ColdFern_Wpn_Uzi_Fire_1P_DryFire
   ]
  }
  nodes [MSS_ColdFern_Foley_lib_0 = &/MSS_ColdFern_Foley_lib(In Array=^Dry Fire, UE.Source.OnPlay=^UE.Source.OnPlay) {also {defaults {Multiplier=0.8}}}]
  links [
   {dst="@UE.OutputFormat.Mono.Audio:0", src=MSS_ColdFern_Foley_lib_0.UE.OutputFormat.Mono.Audio:0}
   {dst="@UE.Source.OneShot.OnFinished", src=MSS_ColdFern_Foley_lib_0.UE.Source.OneShot.OnFinished}
  ]
 }]
properties {
 assetImportData=&self:AssetImportData
 attenuationSettings=&Fort_Weapon_Reload_Att
 bOverrideConcurrency=true
 maxDistance=2250
 sampleRate=48000
 soundClassObject=&Fort_WeaponFoley
 concurrencyOverrides {maxCount=2}
}
