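# MetaSound Source preset for the ColdFern weapon's aim-down-sights-in foley, a mono one-shot.
# It inherits its graph from MSS_ColdFern_Foley_lib and supplies two ADS-in wave variations through the In Array input, alongside a Multiplier input.
# It routes to the weapon foley sound class with reload attenuation, and overrides concurrency to allow a single instance.
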MetaSoundSource /WildEstateWeapons/Gameplay/ColdFern/Sound/Pickups/MSS_ColdFern_Foley_ADS_In
preset_parent=&/MSS_ColdFern_Foley_lib
inputs [
 In Array {type=WaveAsset:Array}
 Multiplier {type=Float}
 UE.Source.OnPlay {type=Trigger}
]
interfaces [UE.OutputFormat.Mono:1.0, UE.Source.OneShot:1.0, UE.Source:1.1]
outputs [
 "UE.OutputFormat.Mono.Audio:0" {type=Audio}
 UE.Source.OneShot.OnFinished {type=Trigger}
]
pages [Default {input_defaults {
   Multiplier=1
   UE.Source.OnPlay=null
   In Array [/WildEstateWeapons/Gameplay/ColdFern/Sound/wav/ColdFern_ADS_In_01, /WildEstateWeapons/Gameplay/ColdFern/Sound/wav/ColdFern_ADS_In_02]
  }}]
properties {
 assetImportData=&self:AssetImportData
 attenuationSettings=&Fort_Weapon_Reload_Att
 bOverrideConcurrency=true
 maxDistance=2250
 sampleRate=48000
 soundClassObject=&Fort_WeaponFoley
 concurrencyOverrides {maxCount=1}
}
