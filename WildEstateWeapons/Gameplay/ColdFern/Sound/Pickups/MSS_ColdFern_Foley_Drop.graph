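# Mono one-shot MetaSound source for the ColdFern weapon's drop foley, set up with the weapon foley sound class and reload attenuation.
# It wraps the shared MSS_ColdFern_Foley_lib graph, passing in Multiplier and returning that graph's OnFinished.
# It declares an array of two drop waves, but the graph shows no connection from that array, OnPlay or the audio output to the library node.
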MetaSoundSource /WildEstateWeapons/Gameplay/ColdFern/Sound/Pickups/MSS_ColdFern_Foley_Drop
inputs [
 In Array {type=WaveAsset:Array}
 Multiplier {type=Float}
 UE.Source.OnPlay {type=Trigger}
]
interfaces [UE.OutputFormat.Mono:1.0, UE.Source.OneShot:1.0, UE.Source:1.1]
outputs [
 "UE.OutputFormat.Mono.Audio:0" {type=Audio}
 UE.Source.OneShot.OnFinished {type=Trigger}
]
pages [Default {
  input_defaults {
   Multiplier=1
   UE.Source.OnPlay=null
   In Array [/WildEstateWeapons/Gameplay/ColdFern/Sound/wav/ColdFern_Drop_01, /WildEstateWeapons/Gameplay/ColdFern/Sound/wav/ColdFern_Drop_02]
  }
  nodes [MSS_ColdFern_Foley_lib_0 = &/MSS_ColdFern_Foley_lib(Multiplier=^Multiplier)]
  links [{dst="@UE.Source.OneShot.OnFinished", src=MSS_ColdFern_Foley_lib_0.UE.Source.OneShot.OnFinished}]
 }]
properties {
 assetImportData=&self:AssetImportData
 attenuationSettings=&Fort_Weapon_Reload_Att
 maxDistance=2250
 sampleRate=48000
 soundClassObject=&Fort_WeaponFoley
}
