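# Mono one-shot MetaSound for the foley impact when a dropped ColdFern weapon pickup lands.
# On play, it picks a random land sound from an array, while a 50% trigger filter decides whether a delayed beer can sound follows.
# Both wave players mix to mono output, and trigger accumulators signal OnFinished correctly for either branch.
# The Multiplier input is not connected.
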MetaSoundSource /WildEstateWeapons/Gameplay/ColdFern/Sound/Pickups/MSS_ColdFern_Foley_Drop_Impact
inputs [
 Beer Can {type=WaveAsset}
 In Array {type=WaveAsset:Array}
 Multiplier {type=Float}
 UE.Source.OnPlay {type=Trigger}
]
interfaces [UE.OutputFormat.Mono:1.0, UE.Source.OneShot:1.0, UE.Source:1.1]
outputs [
 "UE.OutputFormat.Mono.Audio:0" {type=Audio}
 UE.Source.OneShot.OnFinished {type=Trigger}
]
pages [Default {
  input_defaults {
   Beer Can=/WildEstateWeapons/Gameplay/ColdFern/Sound/wav/ColdFern_DropLand_BeerCan
   Multiplier=1
   UE.Source.OnPlay=null
   In Array [/WildEstateWeapons/Gameplay/ColdFern/Sound/wav/ColdFern_DropLand_01, /WildEstateWeapons/Gameplay/ColdFern/Sound/wav/ColdFern_DropLand_02]
  }
  nodes [
   ArrayRandomGet_0 = Array.Random Get:WaveAsset:Array(In Array=^In Array, Next=^UE.Source.OnPlay)
   WavePlayer_0 = UE.Wave Player:Mono(Play=@ArrayRandomGet_0.On Next, Wave Asset=@ArrayRandomGet_0.Value) {also {version [1,1]}}
   TriggerFilter_0 = UE.Trigger Filter(Trigger=^UE.Source.OnPlay) {also {defaults {Probability=0.5}}}
   TriggerDelay_0 = UE.Trigger Delay(In=@TriggerFilter_0.Heads) {also {defaults {Delay Time=0.85}}}
   WavePlayer_1 = UE.Wave Player:Mono(Wave Asset=^Beer Can, Play=@TriggerDelay_0.Out) {also {version [1,1]}}
   AudioMixerAudioMixerMono2_0 = "AudioMixer.Audio Mixer (Mono, 2)"(In 0=@WavePlayer_0.Out Mono, In 1=@WavePlayer_1.Out Mono) {also {defaults {Gain 0=0.75, Gain 1=0.8}}}
   TriggerAccumulatorTriggerAccumulate2_0 = "TriggerAccumulator.Trigger Accumulate (2)"(In 0=UE.Reroute(Value=@TriggerFilter_0.Tails).Value, In 1=@WavePlayer_0.On Finished)
   TriggerAccumulatorTriggerAccumulate2_1 = "TriggerAccumulator.Trigger Accumulate (2)"(In 0=@WavePlayer_0.On Finished, In 1=@WavePlayer_1.On Finished)
   TriggerAnyTriggerAny2_0 = "TriggerAny.Trigger Any (2)"(In 0=@TriggerAccumulatorTriggerAccumulate2_0.Out, In 1=@TriggerAccumulatorTriggerAccumulate2_1.Out)
  ]
  links [
   {dst="@UE.OutputFormat.Mono.Audio:0", src=AudioMixerAudioMixerMono2_0.Out}
   {dst="@UE.Source.OneShot.OnFinished", src=TriggerAnyTriggerAny2_0.Out}
  ]
 }]
properties {
 assetImportData=&self:AssetImportData
 attenuationSettings=&Foley_Attenuation_Basic_Att
 maxDistance=1536
 sampleRate=48000
 soundClassObject=&Fort_WeaponFoley
}
