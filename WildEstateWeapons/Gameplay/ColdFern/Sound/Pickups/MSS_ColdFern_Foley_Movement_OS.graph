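# Mono one-shot MetaSound for ColdFern weapon movement foley, played in the weapon foley sound class with a small concurrency cap.
# On play, chained trigger compares gate playback on character affiliation, then on the Stance input.
# When both pass, a random wave from the movement array plays with a randomized gain.
# A failed check or the wave ending fires OnFinished, so skipped plays end immediately.
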MetaSoundSource /WildEstateWeapons/Gameplay/ColdFern/Sound/Pickups/MSS_ColdFern_Foley_Movement_OS
inputs [
 Character.Affiliation {type=Enum:FortAudioAffiliation}
 Character.Health {type=Float}
 Character.UniqueID {type=Int32}
 In Array {type=WaveAsset:Array}
 Stance {type=Int32}
 UE.Source.OnPlay {type=Trigger}
]
interfaces [Fort.Character:1.0, UE.OutputFormat.Mono:1.0, UE.Source.OneShot:1.0, UE.Source:1.1]
outputs [
 "UE.OutputFormat.Mono.Audio:0" {type=Audio}
 UE.Source.OneShot.OnFinished {type=Trigger}
]
pages [Default {
  input_defaults {
   Character.Affiliation=0
   Character.Health=0
   Character.UniqueID=0
   Stance=2
   UE.Source.OnPlay=null
   In Array [
    /WildEstateWeapons/Gameplay/ColdFern/Sound/wav/ColdFern_Foley_Mvmt_OS_01
    /WildEstateWeapons/Gameplay/ColdFern/Sound/wav/ColdFern_Foley_Mvmt_OS_02
    /WildEstateWeapons/Gameplay/ColdFern/Sound/wav/ColdFern_Foley_Mvmt_OS_03
    /WildEstateWeapons/Gameplay/ColdFern/Sound/wav/ColdFern_Foley_Mvmt_OS_04
    /WildEstateWeapons/Gameplay/ColdFern/Sound/wav/ColdFern_Foley_Mvmt_OS_05
    /WildEstateWeapons/Gameplay/ColdFern/Sound/wav/ColdFern_Foley_Mvmt_OS_06
    /WildEstateWeapons/Gameplay/ColdFern/Sound/wav/ColdFern_Foley_Mvmt_OS_07
    /WildEstateWeapons/Gameplay/ColdFern/Sound/wav/ColdFern_Foley_Mvmt_OS_08
   ]
  }
  nodes [
   n1 = TriggerCompare.Clamp:Int32(Compare=^UE.Source.OnPlay, A=Convert.Int32:Enum:FortAudioAffiliation("Enum:FortAudioAffiliation"=^Character.Affiliation).Int32)
   n2 = UE.Reroute(Value=@n1.True)
   n3 = TriggerCompare.Clamp:Int32(A=^Stance, Compare=@n2.Value) {also {defaults {B=1}}}
   ArrayRandomGet_0 = Array.Random Get:WaveAsset:Array(In Array=^In Array, Next=@n3.True)
   n4 = UE.RandomFloat(Next=@ArrayRandomGet_0.On Next, Reset=UE.Reroute(Value=UE.Reroute(Value=@n2.Value).Value).Value) {also {defaults {Min=0.6}}}
   WavePlayer_0 = UE.Wave Player:Mono(Wave Asset=@ArrayRandomGet_0.Value, Play=@n4.On Next) {also {version [1,1]}}
   Multiply_0 = UE.Multiply:Audio by Float(AdditionalOperands=UE.Reroute(Value=@n4.Value).Value, PrimaryOperand=@WavePlayer_0.Out Mono)
   TriggerAnyTriggerAny3_0 = "TriggerAny.Trigger Any (3)"(In 1=UE.Reroute(Value=@n3.False).Value, In 0=UE.Reroute(Value=@n1.False).Value, In 2=@WavePlayer_0.On Finished)
  ]
  links [{dst="@UE.OutputFormat.Mono.Audio:0", src=Multiply_0.Out}, {dst="@UE.Source.OneShot.OnFinished", src=TriggerAnyTriggerAny3_0.Out}]
 }]
properties {
 assetImportData=&self:AssetImportData
 bOverrideConcurrency=true
 sampleRate=48000
 soundClassObject=&Fort_WeaponFoley
 volume=0.7
 concurrencyOverrides {maxCount=2}
}
