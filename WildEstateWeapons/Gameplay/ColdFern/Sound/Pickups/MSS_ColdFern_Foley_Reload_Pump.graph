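# One-shot stereo MetaSound source for the ColdFern weapon's pump reload foley.
# On play, a random array get picks one of the reload waves and feeds it to a wave player with a slight upward pitch shift.
# Both channels are scaled by a shared Multiplier input, and playback end fires OnFinished.
# It uses the weapon foley sound class, reload attenuation and a stop-oldest concurrency override.
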MetaSoundSource /WildEstateWeapons/Gameplay/ColdFern/Sound/Pickups/MSS_ColdFern_Foley_Reload_Pump
inputs [
 Multiplier {type=Float}
 Reload {type=WaveAsset:Array}
 UE.Source.OnPlay {type=Trigger}
]
interfaces [UE.OutputFormat.Stereo:1.0, UE.Source.OneShot:1.0, UE.Source:1.1]
outputs [
 "UE.OutputFormat.Stereo.Audio:0" {type=Audio}
 "UE.OutputFormat.Stereo.Audio:1" {type=Audio}
 UE.Source.OneShot.OnFinished {type=Trigger}
]
pages [Default {
  input_defaults {
   Multiplier=0.5
   UE.Source.OnPlay=null
   Reload [
    /WildEstateWeapons/Gameplay/ColdFern/Sound/wav/ColdFern_Reload_01
    /WildEstateWeapons/Gameplay/ColdFern/Sound/wav/ColdFern_Reload_02
    /WildEstateWeapons/Gameplay/ColdFern/Sound/wav/ColdFern_Reload_03
    /WildEstateWeapons/Gameplay/ColdFern/Sound/wav/ColdFern_Reload_04
   ]
  }
  nodes [
   ArrayRandomGet_0 = Array.Random Get:WaveAsset:Array(In Array=^Reload, Next=^UE.Source.OnPlay)
   WavePlayer_0 = UE.Wave Player:Stereo(Play=@ArrayRandomGet_0.On Next, Wave Asset=@ArrayRandomGet_0.Value) {also {defaults {Pitch Shift=1}, version [1,1]}}
   Multiply_0 = UE.Multiply:Audio by Float(AdditionalOperands=^Multiplier, PrimaryOperand=@WavePlayer_0.Out Right)
   Multiply_1 = UE.Multiply:Audio by Float(AdditionalOperands=^Multiplier, PrimaryOperand=@WavePlayer_0.Out Left)
  ]
  links [
   {dst="@UE.OutputFormat.Stereo.Audio:1", src=Multiply_0.Out}
   {dst="@UE.OutputFormat.Stereo.Audio:0", src=Multiply_1.Out}
   {dst="@UE.Source.OneShot.OnFinished", src=WavePlayer_0.On Finished}
  ]
 }]
properties {
 assetImportData=&self:AssetImportData
 attenuationSettings=&Fort_Weapon_Reload_Att
 bOverrideConcurrency=true
 maxDistance=2250
 numChannels=2
 outputFormat=Stereo
 sampleRate=48000
 soundClassObject=&Fort_WeaponFoley
 concurrencyOverrides {maxCount=2, resolutionRule=StopOldest, retriggerTime=0.5}
}
