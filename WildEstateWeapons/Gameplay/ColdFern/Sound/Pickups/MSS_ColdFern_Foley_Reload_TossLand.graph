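# One-shot mono MetaSound for the ColdFern weapon's reload foley, playing the impact of the tossed-away item landing.
# On play, a short randomized delay holds the sound back, then a random multiplier is rolled.
# The shared ColdFern foley library patch then plays from five impact variations with that multiplier and a random pitch range.
# Weapon-reload attenuation and a tight stop-oldest concurrency override limit range and overlap.
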MetaSoundSource /WildEstateWeapons/Gameplay/ColdFern/Sound/Pickups/MSS_ColdFern_Foley_Reload_TossLand
inputs [
 Character.Affiliation {type=Enum:FortAudioAffiliation}
 Character.Health {type=Float}
 Character.UniqueID {type=Int32}
 Impact {type=WaveAsset:Array}
 UE.Source.OnPlay {type=Trigger}
]
interfaces [Fort.Character:1.0, UE.OutputFormat.Mono:1.0, UE.Source.OneShot:1.0, UE.Source:1.1]
outputs [
 "UE.OutputFormat.Mono.Audio:0" {type=Audio}
 UE.Source.OneShot.OnFinished {type=Trigger}
]
pages [Default {
  input_defaults {
   Character.Affiliation=2
   Character.Health=0
   Character.UniqueID=1
   UE.Source.OnPlay=null
   Impact [
    /WildEstateWeapons/Gameplay/ColdFern/Sound/wav/ColdFern_TossAway_Impact_01
    /WildEstateWeapons/Gameplay/ColdFern/Sound/wav/ColdFern_TossAway_Impact_02
    /WildEstateWeapons/Gameplay/ColdFern/Sound/wav/ColdFern_TossAway_Impact_03
    /WildEstateWeapons/Gameplay/ColdFern/Sound/wav/ColdFern_TossAway_Impact_04
    /WildEstateWeapons/Gameplay/ColdFern/Sound/wav/ColdFern_TossAway_Impact_05
   ]
  }
  nodes [
   n1 = UE.RandomTime(Next=^UE.Source.OnPlay) {also {defaults {Max=0.6, Min=0.5}}}
   TriggerDelay_0 = UE.Trigger Delay(In=@n1.On Next, Delay Time=@n1.Value)
   n2 = UE.RandomFloat(Next=@TriggerDelay_0.Out) {also {defaults {Max=0.35, Min=0.15}}}
   MSS_ColdFern_Foley_lib_0 = &/MSS_ColdFern_Foley_lib(In Array=^Impact, UE.Source.OnPlay=@n2.On Next, Multiplier=@n2.Value) {also {defaults {Pitch.Max=2, Pitch.Min=-2}}}
  ]
  links [
   {dst="@UE.OutputFormat.Mono.Audio:0", src=MSS_ColdFern_Foley_lib_0.UE.OutputFormat.Mono.Audio:0}
   {dst="@UE.Source.OneShot.OnFinished", src=MSS_ColdFern_Foley_lib_0.UE.Source.OneShot.OnFinished}
  ]
 }]
properties {
 assetImportData=&self:AssetImportData
 attenuationSettings=&Fort_Weapon_Reload_Att
 bOverrideConcurrency=true
 maxDistance=2250
 sampleRate=48000
 soundClassObject=&Fort_WeaponFoley
 concurrencyOverrides {maxCount=2, resolutionRule=StopOldest, retriggerTime=2}
}
